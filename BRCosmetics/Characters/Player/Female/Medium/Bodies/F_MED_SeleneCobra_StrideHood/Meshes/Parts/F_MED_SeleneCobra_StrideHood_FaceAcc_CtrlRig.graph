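# Corrective Control Rig for the face accessory part of the female SeleneCobra StrideHood cosmetic, adjusting hood-hair and back-hat bones as the head moves.
# Spherical pose readers turn head look directions and right upper-arm poses into alpha variables.
# Parent constraints blend those bones between head, neck and upper spine, then ModifyTransforms nodes add offsets weighted by the head alphas.
# Combined head-and-rifle-arm alphas are calculated but drive no transforms here.
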ControlRigBlueprint /BRCosmetics/Characters/Player/Female/Medium/Bodies/F_MED_SeleneCobra_StrideHood/Meshes/Parts/F_MED_SeleneCobra_StrideHood_FaceAcc_CtrlRig
graphs [
 ContainedGraph {nodes [
   n1 = RigVMFunctionEntryNode()
   SequenceExecution_2 = &RigVMFunction_Sequence(ExecuteContext=@n1.ExecuteContext)
   SequenceExecution_2_1 = &RigVMFunction_Sequence(ExecuteContext=@SequenceExecution_2.B)
   SequenceExecution_2_2 = &RigVMFunction_Sequence(ExecuteContext=@SequenceExecution_2_1.B)
   SequenceExecution_2_3 = &RigVMFunction_Sequence(ExecuteContext=@SequenceExecution_2_2.B)
   RigVMFunctionReturnNode(A=@SequenceExecution_2.A, B=@SequenceExecution_2_1.A, C=@SequenceExecution_2_2.A, D=@SequenceExecution_2_3.A, E=@SequenceExecution_2_3.B)
  ]}
 RigVMFunctionLibrary {links [], nodes []}
 RigVMFunction_Sequence_1_ContainedGraph {nodes [
   n1 = RigVMFunctionEntryNode()
   RigVMFunction_Sequence = &RigVMFunction_Sequence(ExecuteContext=@n1.ExecuteContext)
   n2 = &RigVMFunction_Sequence(ExecuteContext=@RigVMFunction_Sequence.B)
   RigVMFunctionReturnNode(A=@RigVMFunction_Sequence.A, B=@n2.A, C=@n2.B)
  ]}
 RigVMModel {nodes [
   RigUnit_BeginExecution = &RigUnit_BeginExecution()
   RigVMFunction_Sequence = RigVMAggregateNode(ExecuteContext=@RigUnit_BeginExecution.ExecutePin)
   SequenceExecution_2 = RigVMAggregateNode(ExecuteContext=@RigVMFunction_Sequence.A)
   n1 = &RigUnit_SphericalPoseReader(
    ActiveRegionScaleFactors=(PositiveWidth=4,NegativeWidth=1,PositiveHeight=3,NegativeHeight=3)
    ActiveRegionSize=0.07
    Debug=(bDrawDebug=false,bDraw2D=False,bDrawLocalAxes=false,DebugScale=25,DebugSegments=20,DebugThickness=0.25)
    DriverAxis=(X=1,Y=0,Z=0)
    DriverItem=(Type=Bone,Name="head")
    FalloffRegionScaleFactors=(PositiveWidth=1,NegativeWidth=0.8,PositiveHeight=1,NegativeHeight=1)
    FalloffSize=0.4
    FlipHeightScaling=False
    FlipWidthScaling=False
    OptionalParentItem=(Type=Bone,Name="spine_05")
    RotationOffset=(X=70,Y=0,Z=-150)
    ExecutePin=@SequenceExecution_2.D
   )
   > RigVMVariableNode(Value=@n1.OutputParam) {variable=head_SPR_look_up_alpha}
   > n2 = &RigUnit_SphericalPoseReader(
    ActiveRegionScaleFactors=(PositiveWidth=1,NegativeWidth=1,PositiveHeight=1,NegativeHeight=1)
    ActiveRegionSize=0.065
    Debug=(bDrawDebug=false,DebugScale=25,DebugSegments=20,DebugThickness=0.25)
    DriverAxis=(X=-1,Y=0,Z=0)
    DriverItem=(Type=Bone,Name="upperarm_r")
    FalloffRegionScaleFactors=(PositiveWidth=1,NegativeWidth=1,PositiveHeight=1,NegativeHeight=1)
    FalloffSize=0.125
    OptionalParentItem=(Type=Bone)
    RotationOffset=(X=60,Y=0,Z=65)
   )
   > RigVMVariableNode(Value=@n2.OutputParam) {variable=upperarm_r_SPR_back_alpha}
   > n3 = &RigUnit_SphericalPoseReader(
    ActiveRegionScaleFactors=(PositiveWidth=1,NegativeWidth=1,PositiveHeight=1,NegativeHeight=1)
    ActiveRegionSize=0.075
    Debug=(bDrawDebug=false,DebugScale=25,DebugSegments=20,DebugThickness=0.25)
    DriverAxis=(X=-1,Y=0,Z=0)
    DriverItem=(Type=Bone,Name="upperarm_r")
    FalloffRegionScaleFactors=(PositiveWidth=1,NegativeWidth=1,PositiveHeight=1,NegativeHeight=1)
    FalloffSize=0.2
    OptionalParentItem=(Type=Bone)
    RotationOffset=(X=-45,Y=15,Z=-20)
   )
   > RigVMVariableNode(Value=@n3.OutputParam) {variable=upperarm_r_rifle_up_alpha}
   n4 = &RigUnit_SphericalPoseReader(
    ActiveRegionScaleFactors=(PositiveWidth=2,NegativeWidth=2,PositiveHeight=2.4,NegativeHeight=0.5)
    ActiveRegionSize=0.1
    Debug=(bDrawDebug=false,DebugScale=25,DebugSegments=20,DebugThickness=0.25)
    DriverAxis=(X=0,Y=1,Z=0)
    DriverItem=(Type=Bone,Name="head")
    FalloffRegionScaleFactors=(PositiveWidth=1,NegativeWidth=1,PositiveHeight=1,NegativeHeight=0.4)
    FalloffSize=0.25
    OptionalParentItem=(Type=Bone)
    RotationOffset=(X=120,Y=0,Z=0)
    ExecutePin=@SequenceExecution_2.B
   )
   > RigVMVariableNode(Value=@n4.OutputParam) {variable=head_SPR_look_left_alpha}
   n5 = &RigUnit_SphericalPoseReader(
    ActiveRegionScaleFactors=(PositiveWidth=2,NegativeWidth=2,PositiveHeight=3,NegativeHeight=0.5)
    ActiveRegionSize=0.1
    Debug=(bDrawDebug=false,DebugScale=25,DebugSegments=20,DebugThickness=0.25)
    DriverAxis=(X=0,Y=1,Z=0)
    DriverItem=(Type=Bone,Name="head")
    FalloffRegionScaleFactors=(PositiveWidth=1,NegativeWidth=1,PositiveHeight=0,NegativeHeight=0.4)
    FalloffSize=0.27
    FlipHeightScaling=true
    FlipWidthScaling=false
    OptionalParentItem=(Type=Bone)
    RotationOffset=(X=60,Y=0,Z=0)
    ExecutePin=@SequenceExecution_2.C
   )
   > RigVMVariableNode(Value=@n5.OutputParam) {variable=head_SPR_look_right_alpha}
   n6 = &RigUnit_SphericalPoseReader(
    ActiveRegionScaleFactors=(PositiveWidth=2,NegativeWidth=0.5,PositiveHeight=2,NegativeHeight=2)
    ActiveRegionSize=0.075
    Debug=(bDrawDebug=false,DebugScale=25,DebugSegments=20,DebugThickness=0.25)
    DriverAxis=(X=0,Y=1,Z=0)
    DriverItem=(Type=Bone,Name="head")
    FalloffRegionScaleFactors=(PositiveWidth=1,NegativeWidth=0.4,PositiveHeight=1,NegativeHeight=1)
    FalloffSize=0.2
    FlipHeightScaling=false
    FlipWidthScaling=true
    OptionalParentItem=(Type=Bone)
    RotationOffset=(X=90,Y=0,Z=12.5)
    ExecutePin=@SequenceExecution_2.A
   )
   > RigVMVariableNode(Value=@n6.OutputParam) {variable=head_SPR_look_down_alpha}
   &RigUnit_ParentConstraint(
    AdvancedSettings=(InterpolationType=Average,RotationOrderForFilter=XZY)
    Child=(Type=Bone,Name="dyn_hat_bk")
    Parents=((Item=(Type=Bone,Name="head"),Weight=1),(Item=(Type=Bone,Name="spine_05"),Weight=1))
    Weight=1
    ExecutePin=@RigVMFunction_Sequence.B
   )
   > &RigUnit_ParentConstraint(
    AdvancedSettings=(InterpolationType=Average,RotationOrderForFilter=XZY)
    Child=(Type=Bone,Name="dyn_hood_hair_fr_l")
    Parents=((Item=(Type=Bone,Name="head"),Weight=0.3),(Item=(Type=Bone,Name="neck_01"),Weight=0.7))
    Weight=1
   )
   > &RigUnit_ParentConstraint(
    AdvancedSettings=(InterpolationType=Average,RotationOrderForFilter=XZY)
    Child=(Type=Bone,Name="dyn_hood_hair_fr_r")
    Parents=((Item=(Type=Bone,Name="head"),Weight=0.3),(Item=(Type=Bone,Name="neck_01"),Weight=0.7))
    Weight=1
   )
   > &RigUnit_ParentConstraint(
    AdvancedSettings=(InterpolationType=Average,RotationOrderForFilter=XZY)
    Child=(Type=Bone,Name="dyn_hood_hair_fr_l_01")
    Parents=((Item=(Type=Bone,Name="head"),Weight=0.3),(Item=(Type=Bone,Name="neck_01"),Weight=0.7))
    Weight=1
   )
   > RigVMVariableNode(Value=&RigVMFunction_MathFloatMul(A=RigVMVariableNode() {variable=head_SPR_look_up_alpha}.Value, B=RigVMVariableNode() {variable=upperarm_r_rifle_up_alpha}.Value) {template="Multiply::Execute(in A,in B,out Result)"}.Result) {variable=look_up_two_hands_rifle_multiplied_alpha}
   > RigVMVariableNode(Value=&RigVMFunction_MathFloatMul(B=RigVMVariableNode() {variable=head_SPR_look_down_alpha}.Value, A=RigVMVariableNode() {variable=upperarm_r_rifle_up_alpha}.Value) {template="Multiply::Execute(in A,in B,out Result)"}.Result) {variable=look_down_two_hands_rifle_multiplied_alpha_0}
   &RigUnit_ModifyTransforms(
    ItemToModify="((Item=(Type=Bone,Name=\"dyn_hood_hair_fr_l\"),Transform=(Rotation=(X=-0.061,Y=-0.0016,Z=0.0261,W=0.9978),Translation=(X=0,Y=0,Z=0),Scale3D=(X=1,Y=1,Z=1))),(Item=(Type=Bone,Name=\"dyn_hood_hair_fr_l_01\"),Transform=(Rotation=(X=0,Y=0,Z=0,W=1),Translation=(X=0, ...(truncated)"
    WeightMinimum=0
    ExecutePin=@RigVMFunction_Sequence.C
    Weight=RigVMVariableNode() {variable=head_SPR_look_left_alpha}.Value
   )
   > &RigUnit_ModifyTransforms(
    ItemToModify="((Item=(Type=Bone,Name=\"dyn_hood_hair_fr_r\"),Transform=(Rotation=(X=0.2132,Y=-0.0376,Z=-0.1695,W=0.9615),Translation=(X=0,Y=0,Z=0),Scale3D=(X=1,Y=1,Z=1))),(Item=(Type=Bone,Name=\"dyn_hood_hair_fr_l_01\"),Transform=(Rotation=(X=-0.1304,Y=-0.0057,Z=0.0432,W=0. ...(truncated)"
    WeightMinimum=0
    Weight=RigVMVariableNode() {variable=head_SPR_look_right_alpha}.Value
   )
   n7 = RigVMVariableNode() {variable=head_SPR_look_up_alpha}
   > &RigUnit_ModifyTransforms(
    ItemToModify=((Item=(Type=Bone,Name="dyn_hat_bk"),Transform=(Rotation=(X=0,Y=0,Z=0,W=1),Translation=(X=-5,Y=-3,Z=0),Scale3D=(X=1,Y=1,Z=1))))
    WeightMinimum=0
    Weight=@n7.Value
   )
   > &RigUnit_ModifyTransforms(
    ItemToModify="((Item=(Type=Bone,Name=\"dyn_hood_hair_fr_l\"),Transform=(Rotation=(X=-0.061,Y=-0.0027,Z=0.0435,W=0.9972),Translation=(X=0,Y=0,Z=0),Scale3D=(X=1,Y=1,Z=1))),(Item=(Type=Bone,Name=\"dyn_hood_hair_fr_l_01\"),Transform=(Rotation=(X=-0.087,Y=-0.0053,Z=0.0608,W=0.99 ...(truncated)"
    WeightMinimum=0
    Weight=@n7.Value
   )
   > &RigUnit_ModifyTransforms(
    ItemToModify="((Item=(Type=Bone,Name=\"dyn_hood_hair_fr_l\"),Transform=(Rotation=(X=-0.1304,Y=-0.0057,Z=0.0432,W=0.9905),Translation=(X=0,Y=0,Z=0),Scale3D=(X=1,Y=1,Z=1))),(Item=(Type=Bone,Name=\"dyn_hood_hair_fr_l_01\"),Transform=(Rotation=(X=-0.061,Y=-0.0027,Z=0.0435,W=0.9 ...(truncated)"
    WeightMinimum=0
    Weight=RigVMVariableNode() {variable=head_SPR_look_down_alpha}.Value
   )
   RigVMVariableNode(ExecuteContext=()) {variable=look_down_upperarm_r_back_multiplied_alpha}
  ]}
]
hierarchy [
 root {loc=<0 0 0>, rot=<0 0 0 1>}
 pelvis {parent=root, loc=<0 -0.5301 98.0571>, rot=<0.0127 0.707 -0.0127 -0.707>}
 spine_01 {parent=pelvis, loc=<6.1208 -0.9328 0>, rot=<0 0 0.034 -0.9994>}
 spine_02 {parent=spine_01, loc=<6.5996 -0.2095 0>, rot=<0 0 0.0238 -0.9997>}
 spine_03 {parent=spine_02, loc=<6.4914 -0.1237 0>, rot=<0 0 0.0149 -0.9999>}
 spine_04 {parent=spine_03, loc=<6.9598 0.0794 0>, rot=<0 0 -0.0132 -0.9999>}
 spine_05 {parent=spine_04, loc=<6.9987 0 0>, rot=<0 0 -0.0015 -1>}
 neck_01 {parent=spine_05, loc=<11.3138 0.1529 0>, rot=<0 0 -0.2003 -0.9797>}
 neck_02 {parent=neck_01, loc=<4.5627 0 0>, rot=<0 0 0.0504 -0.9987>}
 head {parent=neck_02, loc=<5.3914 0 0>, rot=<0 0 0.115 -0.9934>}
 dyn_frStrand_l_01 {parent=head, loc=<9.5289 7.6914 -3.0816>, rot=<0.1628 -0.9051 -0.3139 -0.236>}
 dyn_frStrand_l_02 {parent=dyn_frStrand_l_01, loc=<3.8467 0 0>, rot=<0 -0.1249 -0.1072 -0.9864>}
 dyn_frStrand_l_03 {parent=dyn_frStrand_l_02, loc=<3.3972 0 0>, rot=<0 0 0 1>}
 dyn_frStrand_l_01_a {parent=dyn_frStrand_l_01, loc=<5.8109 -3.5871 -0.0602>, rot=<0.6023 0.0091 -0.2855 -0.7454>}
 dyn_hatLeft {parent=head, loc=<-2.156 -2.69 -8.386>, rot=<-0.3504 -0.6142 -0.5767 -0.4091>}
 dyn_hatRight {parent=head, loc=<-2.156 -2.69 7.914>, rot=<0.4494 0.5459 0.5713 0.4166>}
 dyn_hatTop {parent=head, loc=<12.7986 2.1557 0.3048>, rot=<-0.4478 -0.5473 -0.5111 -0.4886>}
 dyn_hatTopBack {parent=head, loc=<13.1291 -0.315 0.4974>, rot=<0.5473 -0.4478 0.4886 -0.5111>}
 dyn_hat_bk {parent=head, loc=<1.1592 -7.4792 0>, rot=<0 -1 0 0>}
 dyn_headFront_01_Base_r {parent=head, loc=<3.953 1.6741 6.5253>, rot=<0.0631 -0.884 0.4623 0.0292>}
 dyn_headFront_01_r {parent=dyn_headFront_01_Base_r, loc=<0 0 0>, rot=<0.9177 0.0617 0.0263 0.3915>}
 dyn_headFront_01_Base_l {parent=head, loc=<3.2641 2.5389 -6.0881>, rot=<-0.1444 0.6533 0.7368 -0.0971>}
 dyn_headFront_01_l {parent=dyn_headFront_01_Base_l, loc=<0 0 0>, rot=<-0.6634 0.1086 -0.1309 -0.7287>}
 dyn_headFront_02_l {parent=dyn_headFront_01_l, loc=<0 0 0>, rot=<0 0 0 1>}
 dyn_headFront_03_l {parent=dyn_headFront_02_l, loc=<0 0 0>, rot=<0 0 0 1>}
 dyn_headFront_04_l {parent=dyn_headFront_03_l, loc=<0 0 0>, rot=<0 0 0 1>}
 dyn_head_01_Base_l {parent=head, loc=<3.1271 -1.7629 -6.5169>, rot=<0.0456 -0.7633 -0.6422 0.0531>}
 dyn_head_01_l {parent=dyn_head_01_Base_l, loc=<0 0 0>, rot=<0 0 0 1>}
 dyn_head_02_l {parent=dyn_head_01_l, loc=<5.2189 0.1587 0.9066>, rot=<0.0863 0.0823 0.1845 -0.9756>}
 dyn_head_03_l {parent=dyn_head_02_l, loc=<4.1111 0 0>, rot=<0 0 0 1>}
 dyn_head_02_l_a {parent=dyn_head_02_l, loc=<2.1296 -0.4526 2.0407>, rot=<0.5812 0.1059 -0.1447 -0.7938>}
 dyn_head_01_Base_r {parent=head, loc=<2.3393 -2.5791 6.7725>, rot=<0.066 -0.704 0.7045 0.0602>}
 dyn_head_01_r {parent=dyn_head_01_Base_r, loc=<0 0 0>, rot=<0 0 0 1>}
 dyn_head_02_r {parent=dyn_head_01_r, loc=<5.3391 0 0>, rot=<0.0786 0.0273 0.1549 -0.9844>}
 dyn_head_03_r {parent=dyn_head_02_r, loc=<4.1111 0 0>, rot=<0 0 0 1>}
 dyn_headBack_Lookat {parent=head, loc=<6.9539 -5.7482 -0.0677>, rot=<-0.0528 -0.9986 0 0.0006>}
 dyn_headBack_Base {parent=dyn_headBack_Lookat, loc=<0 0 0>, rot=<-0.0001 0.0006 0.2324 -0.9726>}
 dyn_headBack_01_c {parent=dyn_headBack_Base, loc=<5.269 -0.4697 -0.1062>, rot=<0 0 -0.0203 -0.9998>}
 dyn_headBack_02_c {parent=dyn_headBack_01_c, loc=<3.9528 0 0>, rot=<0 0 -0.0156 -0.9999>}
 dyn_headBack_03_c {parent=dyn_headBack_02_c, loc=<4.6852 0 0>, rot=<0 0 0 1>}
 dyn_headBack_01_l {parent=dyn_headBack_Base, loc=<4.8669 1.4907 5.4464>, rot=<0.4407 0.2004 0.0062 -0.875>}
 dyn_headBack_02_l {parent=dyn_headBack_01_l, loc=<4.1345 0 0>, rot=<-0.0393 -0.0144 0.0264 -0.9988>}
 dyn_headBack_03_l {parent=dyn_headBack_02_l, loc=<4.5827 0 0>, rot=<0 0 0 1>}
 dyn_headBack_01_r {parent=dyn_headBack_Base, loc=<4.6665 2.2033 -5.0424>, rot=<-0.1882 -0.0662 -0.0164 -0.9798>}
 dyn_headBack_02_r {parent=dyn_headBack_01_r, loc=<4.1345 0 0>, rot=<-0.1285 -0.0538 0.0742 -0.9875>}
 dyn_headBack_03_r {parent=dyn_headBack_02_r, loc=<4.5827 0 0>, rot=<0 0 0 1>}
 dyn_fr_02 {parent=head, loc=<6.6351 9.6486 -0.5862>, rot=<0.0006 -0.1499 0.9887 -0.0041>}
 dyn_fr_01 {parent=head, loc=<9.9761 8.7429 -1.9214>, rot=<0.0009 -0.2113 0.9774 -0.004>}
 dyn_fr_03 {parent=head, loc=<14.1696 3.7909 -0.7952>, rot=<-0.0006 0.1564 0.9877 -0.0041>}
 dyn_fr_04 {parent=head, loc=<5.9605 3.8917 7.358>, rot=<0.0026 -0.6266 0.7793 -0.0032>}
 dyn_fr_05 {parent=head, loc=<7.7784 4.1438 7.4531>, rot=<0.0022 -0.5264 0.8503 -0.0035>}
 dyn_fr_06 {parent=head, loc=<10.6531 6.1085 4.2828>, rot=<0.0057 -0.5203 0.854 -0.0011>}
 dyn_fr_07 {parent=head, loc=<12.4355 5.8022 3.3535>, rot=<0.002 -0.4792 0.8777 -0.0036>}
 dyn_fr_08 {parent=head, loc=<13.8037 4.6991 1.9552>, rot=<0.0015 -0.3579 0.9337 -0.0039>}
 dyn_top_06 {parent=head, loc=<13.4074 -5.5652 -2.135>, rot=<0.0041 -0.9866 -0.1632 0.0007>}
 dyn_Hair_extra_08 {parent=head, loc=<-2.6797 0.4743 5.7077>, rot=<0.0015 -0.371 0.9286 -0.0038>}
 dyn_hood_hair_fr_l {parent=head, loc=<-5.0705 2.2091 -4.6127>, rot=<0.0029 -0.7071 -0.7071 0.0029>}
 dyn_hood_hair_fr_r {parent=head, loc=<-4.8109 2.8915 5.0614>, rot=<0.0029 -0.7071 -0.7071 0.0029>}
 dyn_hood_hair_fr_l_01 {parent=head, loc=<-3.1199 2.876 -5.3936>, rot=<-0.0029 0.7071 0.7071 -0.0029>}
 dyn_mask {parent=head, loc=<-0.0881 9.8875 0>, rot=<0.0029 -0.7071 -0.0029 -0.7071>}
 clavicle_l {parent=spine_05, loc=<5.4831 -0.1053 -3.2946>, rot=<0.0465 -0.7289 -0.009 -0.6829>}
 upperarm_l {parent=clavicle_l, loc=<9.8658 -1.2136 -0.4105>, rot=<0.0164 -0.3245 -0.0269 -0.9454>}
 upperarm_twist_01_l {parent=upperarm_l, loc=<7.2135 -0.1134 0>, rot=<0 0 0 1>}
 deform_upperarm_bk_base_l {parent=upperarm_twist_01_l, loc=<-6.4053 -4.3746 0.314>, rot=<0 0 0 -1>}
 deform_upperarm_bk_l {parent=deform_upperarm_bk_base_l, loc=<0 0 0>, rot=<0 0 0 1>}
 deform_upperarm_fr_base_l {parent=upperarm_twist_01_l, loc=<-5.6312 5.092 0.3354>, rot=<0 0 0 -1>}
 deform_upperarm_fr_l {parent=deform_upperarm_fr_base_l, loc=<0 0 0>, rot=<0 0 0 1>}
 deform_upperarm_up_base_l {parent=upperarm_twist_01_l, loc=<-8.891 0.4825 3.9517>, rot=<0 0 0 -1>}
 deform_upperarm_up_l {parent=deform_upperarm_up_base_l, loc=<0 0 0>, rot=<0 0 0 1>}
 clavicle_r {parent=spine_05, loc=<5.4831 -0.1053 3.2946>, rot=<0.7289 0.0465 -0.6829 0.009>}
 upperarm_r {parent=clavicle_r, loc=<-9.8658 1.2136 0.4105>, rot=<0.0164 -0.3245 -0.0269 -0.9454>}
 upperarm_twist_01_r {parent=upperarm_r, loc=<-7.2135 0.1134 0>, rot=<0 0 0 -1>}
 deform_upperarm_bk_base_r {parent=upperarm_twist_01_r, loc=<6.4053 4.3746 -0.314>, rot=<0 0 0 -1>}
 deform_upperarm_bk_r {parent=deform_upperarm_bk_base_r, loc=<0 0 0>, rot=<0 0 0 1>}
 deform_upperarm_fr_base_r {parent=upperarm_twist_01_r, loc=<5.6312 -5.092 -0.3354>, rot=<0 0 0 -1>}
 deform_upperarm_fr_r {parent=deform_upperarm_fr_base_r, loc=<0 0 0>, rot=<0 0 0 1>}
 deform_upperarm_up_base_r {parent=upperarm_twist_01_r, loc=<8.891 -0.4825 -3.9517>, rot=<0 0 0 -1>}
 deform_upperarm_up_r {parent=deform_upperarm_up_base_r, loc=<0 0 0>, rot=<0 0 0 1>}
]
variables [
 head_SPR_look_down_alpha {type=double, default=0}
 head_SPR_look_left_alpha {type=double, default=0}
 head_SPR_look_right_alpha {type=double, default=0}
 head_SPR_look_up_alpha {type=double, default=0}
 look_down_two_hands_rifle_multiplied_alpha_0 {type=double, default=0}
 look_down_upperarm_r_back_multiplied_alpha {type=double, default=0}
 look_up_two_hands_rifle_multiplied_alpha {type=double, default=0}
 upperarm_r_SPR_back_alpha {type=double, default=0}
 upperarm_r_rifle_up_alpha {type=double, default=0}
]
